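# Sound cue for the security camera alarm among the Security Turret gadget sounds, in the Fort_Gadgets sound class.
# A mixer plays a three-beep one-shot at full input volume alongside a quieter looping alarm.
# An enveloper holds the loop silent for just under two seconds, then ramps it up sharply, so the beeps lead into the loop.
# Attenuation uses a dedicated alarm asset, and a concurrency override caps simultaneous instances.
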SoundCue /FNGameplayCues/Athena/Sounds/Gadgets/SecurityTurret/Security_Camera_Alarm_Loop_Cue
root=Mixer_1
nodes [
 {
  id=Enveloper_1
  children [WavePlayer_3]
  properties {
   pitchCurve {editorCurveData {keys [{time=0}]}}
   volumeCurve {editorCurveData {keys [<at 0 0>, {interpMode=RCIM_Cubic, leaveTangent=0.5, time=1.85, value=0}, <at 1.89 1>]}}
  }
 }
 {id=Mixer_1, children [Enveloper_1,WavePlayer_2], properties {inputVolume [0.4,1]}}
 {id=WavePlayer_2, properties {soundWaveAssetPtr=&/Security_Camera_Alarm_3_Beeps_01}}
 {
  id=WavePlayer_3
  properties {bLooping=true, soundWaveAssetPtr=&/Security_Camera_Alarm_Loop_04}
 }
]
properties {
 attenuationSettings=&/Security_Camera_Alarm_Light_Att
 bOverrideConcurrency=true
 maxDistance=6000
 soundClassObject=&Fort_Gadgets
 volumeMultiplier=1.2
 attenuationOverrides {distanceAlgorithm=Custom, falloffDistance=10112.3027, attenuationShapeExtents {x=600}}
 concurrencyOverrides {maxCount=3}
}
