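# Sound cue for the start of the security camera alarm, filed under security turret gadget sounds, with an empty node graph and no root, so it references no audio.
# It carries only playback settings: the Fort_Gadgets sound class, reduced volume and the alarm light attenuation asset, alongside override values for falloff and shape.
# Its concurrency override values allow one instance per owner and block new plays rather than cutting off the active one.
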SoundCue /FNGameplayCues/Athena/Sounds/Gadgets/SecurityTurret/Security_Camera_Alarm_Start_Cue
root=null
nodes []
properties {
 attenuationSettings=&/Security_Camera_Alarm_Light_Att
 maxDistance=4300
 soundClassObject=&Fort_Gadgets
 volumeMultiplier=0.7
 attenuationOverrides {falloffDistance=2400, attenuationShapeExtents {x=600}}
 concurrencyOverrides {bLimitToOwner=true, maxCount=1, resolutionRule=PreventNew}
}
