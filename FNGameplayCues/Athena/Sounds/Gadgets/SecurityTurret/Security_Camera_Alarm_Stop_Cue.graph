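# Sound cue for the Athena security turret gadget, playing a security camera alarm stop sound in the gadgets sound class.
# A random node picks one of two equally weighted waves and feeds an attenuation node that uses the camera movement attenuation asset.
# That branch enters a mixer at reduced volume, while the mixer's second input is empty.
# Concurrency override settings specify one instance per owner, preventing new plays.
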SoundCue /FNGameplayCues/Athena/Sounds/Gadgets/SecurityTurret/Security_Camera_Alarm_Stop_Cue
root=Mixer_0
nodes [
 {
  id=Attenuation_0
  children [Random_0]
  properties {attenuationSettings=/Game/Athena/Sounds/Gadgets/SecurityTurret/Security_Camera_Movement_Att}
 }
 {id=Mixer_0, children [Attenuation_0, null], properties {inputVolume [0.5,1.2]}}
 {id=Random_0, children [WavePlayer_1,WavePlayer_2], properties {weights [1,1]}}
 {id=WavePlayer_1, properties {soundWaveAssetPtr=&/Security_Camera_Alarm_Stop_01}}
 {id=WavePlayer_2, properties {soundWaveAssetPtr=&/Security_Camera_Alarm_Stop_02}}
]
properties {
 maxDistance=2000
 soundClassObject=&Fort_Gadgets
 volumeMultiplier=1
 concurrencyOverrides {bLimitToOwner=true, maxCount=1, resolutionRule=PreventNew}
}
