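# Sound cue for a security turret camera being destroyed, in the gadgets sound class.
# A mixer layers a random pick of three power-off variations with a random pick of three destroy variations delayed by a short fixed offset.
# Overridden attenuation adds occlusion with low-pass filtering, and the cue uses a shared object-impacts concurrency set to limit overlapping instances.
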SoundCue /FNGameplayCues/Athena/Sounds/Gadgets/SecurityTurret/Security_Camera_Destroy_Cue
root=Mixer_0
nodes [
 {id=Delay_0, children [Random_2], properties {delayMax=0.05, delayMin=0.05}}
 {id=Mixer_0, children [Delay_0,Random_0], properties {inputVolume [1,1]}}
 {id=Random_0, children [WavePlayer_1,WavePlayer_2,WavePlayer_3], properties {bShouldExcludeFromBranchCulling=true, weights [1,1,1]}}
 {id=Random_2, children [WavePlayer_0,WavePlayer_7,WavePlayer_8], properties {bShouldExcludeFromBranchCulling=true, weights [1,1,1]}}
 {id=WavePlayer_0, properties {soundWaveAssetPtr=&/Security_Camera_Destroy_01}}
 {id=WavePlayer_1, properties {soundWaveAssetPtr=&/Security_Camera_PowerOff_01}}
 {id=WavePlayer_2, properties {soundWaveAssetPtr=&/Security_Camera_PowerOff_02}}
 {id=WavePlayer_3, properties {soundWaveAssetPtr=&/Security_Camera_PowerOff_03}}
 {id=WavePlayer_7, properties {soundWaveAssetPtr=&/Security_Camera_Destroy_02}}
 {id=WavePlayer_8, properties {soundWaveAssetPtr=&/Security_Camera_Destroy_03}}
]
properties {
 bOverrideAttenuation=true
 maxDistance=3100
 soundClassObject=&Fort_Gadgets
 volumeMultiplier=0.8
 attenuationOverrides {
  bEnableOcclusion=true
  falloffDistance=2400
  occlusionLowPassFilterFrequency=6000
  occlusionTraceChannel=ECC_GameTraceChannel11
  attenuationShapeExtents {x=700}
 }
 concurrencyOverrides {maxCount=3, resolutionRule=StopOldest}
 concurrencySet [&CONC_Object_Impacts_FourMax]
}
