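# Sound cue for the discreet variant of the security camera alarm-stop sound, under the Security Turret gadget audio.
# A Random node picks between two equally weighted alarm-stop waves and is excluded from branch culling.
# The cue plays them very quietly at raised pitch, through the Fort_Gadgets sound class and a dedicated discreet-alarm attenuation asset.
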SoundCue /FNGameplayCues/Athena/Sounds/Gadgets/SecurityTurret/Security_Camera_Discreet_Alarm_Stop_Cue
root=Random_0
nodes [
 {id=Random_0, children [WavePlayer_1,WavePlayer_2], properties {bShouldExcludeFromBranchCulling=true, weights [1,1]}}
 {id=WavePlayer_1, properties {soundWaveAssetPtr=&/Security_Camera_Alarm_Stop_01}}
 {id=WavePlayer_2, properties {soundWaveAssetPtr=&/Security_Camera_Alarm_Stop_02}}
]
properties {
 attenuationSettings=&/Security_Camera_Discrete_Alarm_Att
 maxDistance=6000
 pitchMultiplier=2
 soundClassObject=&Fort_Gadgets
 volumeMultiplier=0.1
 concurrencyOverrides {bLimitToOwner=true, maxCount=1, resolutionRule=PreventNew}
}
