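# Sound cue named for the power-off sound of the discreet security camera gadget, filed with Athena security turret gameplay cue audio.
# It has no sound nodes and a null root, so it references no audio and carries only playback settings.
# Those settings route it to the gadgets sound class with the camera movement attenuation, reduced volume and raised pitch.
# A concurrency override limits it to one instance per owner, stopping the oldest.
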SoundCue /FNGameplayCues/Athena/Sounds/Gadgets/SecurityTurret/Security_Camera_Discreet_PowerOff_Cue
root=null
nodes []
properties {
 attenuationSettings=/Game/Athena/Sounds/Gadgets/SecurityTurret/Security_Camera_Discreet_Movement_Att
 bOverrideConcurrency=true
 maxDistance=1850
 pitchMultiplier=1.1
 soundClassObject=&Fort_Gadgets
 volumeMultiplier=0.5
 concurrencyOverrides {bLimitToOwner=true, maxCount=1, resolutionRule=StopOldest}
}
