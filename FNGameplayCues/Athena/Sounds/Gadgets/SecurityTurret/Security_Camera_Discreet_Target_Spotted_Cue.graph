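# Sound cue named for the security turret gadget's discreet target-spotted alert, assigned to the Fort_Gadgets sound class.
# It has no sound nodes and a null root, so it has nothing to play as shown.
# Its concurrency override limits it to one instance per owner and blocks new plays while one is active.
# It also stores attenuation settings with listener focus and falloff, though no attenuation override flag is shown.
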SoundCue /FNGameplayCues/Athena/Sounds/Gadgets/SecurityTurret/Security_Camera_Discreet_Target_Spotted_Cue
root=null
nodes []
properties {
 bOverrideConcurrency=true
 maxDistance=2097152
 soundClassObject=&Fort_Gadgets
 volumeMultiplier=1
 attenuationOverrides {
  bEnableFocusInterpolation=true
  bEnableListenerFocus=true
  falloffDistance=1200
  nonFocusVolumeAttenuation=0.5
  stereoSpread=500
  attenuationShapeExtents {x=0}
 }
 concurrencyOverrides {bLimitToOwner=true, maxCount=1, resolutionRule=PreventNew}
}
