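# Sound cue for impacts on the security camera, one of the Athena security turret gadget sounds.
# A random node picks one of five equally weighted impact waves and passes it through a modulator at the root for variation.
# It uses the bullet impact attenuation and the gadgets sound class, with reduced volume and a concurrency override that blocks new instances once the limit is reached.
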SoundCue /FNGameplayCues/Athena/Sounds/Gadgets/SecurityTurret/Security_Camera_Impact_Cue
root=Modulator_0
nodes [
 {id=Modulator_0, children [Random_2]}
 {id=Random_2, children [WavePlayer_3, WavePlayer_8, WavePlayer_9, WavePlayer_10, WavePlayer_11], properties {weights [1,1,1,1,1]}}
 {id=WavePlayer_10, properties {soundWaveAssetPtr=&/Security_Camera_Impact_04}}
 {id=WavePlayer_11, properties {soundWaveAssetPtr=&/Security_Camera_Impact_05}}
 {id=WavePlayer_3, properties {soundWaveAssetPtr=&/Security_Camera_Impact_01}}
 {id=WavePlayer_8, properties {soundWaveAssetPtr=&/Security_Camera_Impact_02}}
 {id=WavePlayer_9, properties {soundWaveAssetPtr=&/Security_Camera_Impact_03}}
]
properties {
 attenuationSettings=/Game/Sounds/Attenuation/Fort_Bullet_Impact_Athena_Att
 bOverrideConcurrency=true
 maxDistance=4500
 soundClassObject=&Fort_Gadgets
 volumeMultiplier=0.6
 concurrencyOverrides {resolutionRule=PreventNew}
}
