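# Sound cue for a security camera powering off, filed with the Security Turret gadget sounds.
# A Random node picks one of three equally weighted power-off waves and passes it through a Modulator at the root.
# Playback uses the turret's spherical attenuation and the gadget sound class, with slightly raised pitch and boosted volume.
# Overridden concurrency limits it to one instance per owner, stopping the oldest.
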SoundCue /FNGameplayCues/Athena/Sounds/Gadgets/SecurityTurret/Security_Camera_PowerOff_Cue
root=Modulator_1
nodes [
 {id=Modulator_1, children [Random_1]}
 {id=Random_1, children [WavePlayer_1,WavePlayer_4,WavePlayer_5], properties {weights [1,1,1]}}
 {id=WavePlayer_1, properties {soundWaveAssetPtr=&/Security_Camera_PowerOff_01}}
 {id=WavePlayer_4, properties {soundWaveAssetPtr=&/Security_Camera_PowerOff_02}}
 {id=WavePlayer_5, properties {soundWaveAssetPtr=&/Security_Camera_PowerOff_03}}
]
properties {
 attenuationSettings=/Game/Athena/Sounds/Gadgets/SecurityTurret/Security_Turret_Sphere_Att
 bOverrideConcurrency=true
 maxDistance=3500
 pitchMultiplier=1.1
 soundClassObject=&Fort_Gadgets
 volumeMultiplier=1.5
 concurrencyOverrides {bLimitToOwner=true, maxCount=1, resolutionRule=StopOldest}
}
