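# Sound cue for a security camera spotting a target, in the gadgets sound class.
# A mixer layers a Target_Lost_03 wave, despite the cue's Spotted name, with a quieter, delayed voice line picked at random from four Target Found variations.
# Each layer has its own attenuation, using turret sphere and vox settings respectively.
# Overridden concurrency allows one instance per owner and blocks new plays while one is active.
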SoundCue /FNGameplayCues/Athena/Sounds/Gadgets/SecurityTurret/Security_Camera_Target_Spotted_Cue
root=Mixer_0
nodes [
 {
  id=Attenuation_0
  children [WavePlayer_7]
  properties {attenuationSettings=/Game/Athena/Sounds/Gadgets/SecurityTurret/Security_Turret_Sphere_Att}
 }
 {id=Attenuation_1, children [Delay_1], properties {attenuationSettings=/Game/Athena/Sounds/Gadgets/SecurityTurret/Security_Turret_Vox_Att}}
 {id=Delay_1, children [Random_0], properties {delayMax=0.8, delayMin=0.8}}
 {id=Mixer_0, children [Attenuation_0, Attenuation_1], properties {inputVolume [1.25,0.8]}}
 {id=Random_0, children [WavePlayer_2,WavePlayer_3,WavePlayer_4,WavePlayer_5], properties {weights [1,1,1,1]}}
 {id=WavePlayer_2, properties {soundWaveAssetPtr=&/Security_Camera_Vox_Target_Found_01}}
 {id=WavePlayer_3, properties {soundWaveAssetPtr=&/Security_Camera_Vox_Target_Found_02}}
 {id=WavePlayer_4, properties {soundWaveAssetPtr=&/Security_Camera_Vox_Target_Found_03}}
 {id=WavePlayer_5, properties {soundWaveAssetPtr=&/Security_Camera_Vox_Target_Found_04}}
 {id=WavePlayer_7, properties {soundWaveAssetPtr=&/Security_Camera_Target_Lost_03}}
]
properties {
 bOverrideConcurrency=true
 maxDistance=3500
 soundClassObject=&Fort_Gadgets
 volumeMultiplier=1
 attenuationOverrides {
  bEnableFocusInterpolation=true
  bEnableListenerFocus=true
  falloffDistance=1200
  nonFocusVolumeAttenuation=0.5
  stereoSpread=500
  attenuationShapeExtents {x=0}
 }
 concurrencyOverrides {bLimitToOwner=true, maxCount=1, resolutionRule=PreventNew}
}
