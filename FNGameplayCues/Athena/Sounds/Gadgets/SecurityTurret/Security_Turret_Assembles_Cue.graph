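# Sound cue for the security turret gadget's assembly sound in Athena.
# It plays a single wave through a short fixed delay, routed to the gadgets sound class with reduced volume and the turret's shared sphere attenuation.
# A concurrency override allows one instance per owner, stopping the oldest when a new one starts.
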SoundCue /FNGameplayCues/Athena/Sounds/Gadgets/SecurityTurret/Security_Turret_Assembles_Cue
root=Delay_1
nodes [
 {id=Delay_1, children [WavePlayer_0], properties {delayMax=0.04, delayMin=0.04}}
 {id=WavePlayer_0, properties {soundWaveAssetPtr=&/Security_Turret_Assembles_01}}
]
properties {
 attenuationSettings=/Game/Athena/Sounds/Gadgets/SecurityTurret/Security_Turret_Sphere_Att
 bOverrideConcurrency=true
 maxDistance=3500
 soundClassObject=&Fort_Gadgets
 volumeMultiplier=0.4
 concurrencyOverrides {bLimitToOwner=true, maxCount=1, resolutionRule=StopOldest}
}
