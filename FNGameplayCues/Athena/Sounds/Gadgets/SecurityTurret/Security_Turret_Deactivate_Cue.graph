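# Sound cue for the Security Turret gadget's deactivation, playing a disassembly sound.
# A single wave player feeds the output directly, at reduced volume and slightly raised pitch, routed to the gadgets sound class with the turret's sphere attenuation asset.
# A concurrency override limits playback to one instance per owner, stopping the oldest when a new one starts.
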SoundCue /FNGameplayCues/Athena/Sounds/Gadgets/SecurityTurret/Security_Turret_Deactivate_Cue
root=WavePlayer_0
nodes [{id=WavePlayer_0, properties {soundWaveAssetPtr=/Game/Athena/Sounds/Gadgets/SecurityTurret/Security_Turret_Disassembles_01}}]
properties {
 attenuationSettings=/Game/Athena/Sounds/Gadgets/SecurityTurret/Security_Turret_Sphere_Att
 bOverrideConcurrency=true
 maxDistance=3500
 pitchMultiplier=1.1
 soundClassObject=&Fort_Gadgets
 volumeMultiplier=0.4
 concurrencyOverrides {bLimitToOwner=true, maxCount=1, resolutionRule=StopOldest}
}
