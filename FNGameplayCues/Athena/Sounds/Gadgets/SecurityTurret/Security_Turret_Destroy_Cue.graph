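# Sound cue for a security turret gadget being destroyed in Athena, routed to the medium explosives sound class with turret explosion attenuation.
# A random one of three Destroy waves plays immediately, mixed with a half-volume random DamageDestroy wave after a short random delay.
# The mix passes through a modulator at the root.
# A concurrency override limits simultaneous instances to three, stopping the oldest.
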SoundCue /FNGameplayCues/Athena/Sounds/Gadgets/SecurityTurret/Security_Turret_Destroy_Cue
root=Modulator_1
nodes [
 {id=Delay_0, children [Random_0], properties {delayMax=0.3, delayMin=0.15}}
 {id=Mixer_0, children [Random_2,Delay_0], properties {inputVolume [1,0.5]}}
 {id=Modulator_1, children [Mixer_0]}
 {id=Random_0, children [WavePlayer_1,WavePlayer_2,WavePlayer_3], properties {weights [1,1,1]}}
 {id=Random_2, children [WavePlayer_0,WavePlayer_7,WavePlayer_8], properties {weights [1,1,1]}}
 {id=WavePlayer_0, properties {soundWaveAssetPtr=&/Security_Turret_Destroy_01}}
 {id=WavePlayer_1, properties {soundWaveAssetPtr=&/Security_Turret_DamageDestroy_01}}
 {id=WavePlayer_2, properties {soundWaveAssetPtr=&/Security_Turret_DamageDestroy_02}}
 {id=WavePlayer_3, properties {soundWaveAssetPtr=&/Security_Turret_DamageDestroy_03}}
 {id=WavePlayer_7, properties {soundWaveAssetPtr=&/Security_Turret_Destroy_02}}
 {id=WavePlayer_8, properties {soundWaveAssetPtr=&/Security_Turret_Destroy_03}}
]
properties {
 attenuationSettings=/FNGameplayCues/Sounds/Attenuation/ATT_Explosion_Turret
 bOverrideConcurrency=true
 maxDistance=27500
 soundClassObject=/Game/Sounds/FortSoundClassesAndModes/Fort_Explosives_Medium
 volumeMultiplier=0.9
 attenuationOverrides {
  bEnableOcclusion=true
  falloffDistance=2700
  occlusionInterpolationTime=1
  occlusionLowPassFilterFrequency=10000
  occlusionTraceChannel=ECC_GameTraceChannel11
  attenuationShapeExtents {x=1100}
 }
 concurrencyOverrides {maxCount=3, resolutionRule=StopOldest, volumeScaleAttackTime=0.1}
}
