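# Sound cue for the security turret's firing build-up, playing a single charge-up wave in the gadgets sound class.
# A fixed short delay node holds back the wave player before it sounds.
# A shared turret sphere attenuation asset handles spatial falloff.
# A concurrency override allows one instance per owner and blocks new plays while it is active.
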SoundCue /FNGameplayCues/Athena/Sounds/Gadgets/SecurityTurret/Security_Turret_Fire_BuildUp_Cue
root=Delay_0
nodes [
 {id=Delay_0, children [WavePlayer_0], properties {delayMax=0.1, delayMin=0.1}}
 {id=WavePlayer_0, properties {soundWaveAssetPtr=&/Security_Turret_ChargeUp_01}}
]
properties {
 attenuationSettings=/Game/Athena/Sounds/Gadgets/SecurityTurret/Security_Turret_Sphere_Att
 bOverrideConcurrency=true
 maxDistance=3500
 soundClassObject=&Fort_Gadgets
 volumeMultiplier=1.25
 attenuationOverrides {
  bEnableFocusInterpolation=true
  bEnableListenerFocus=true
  falloffDistance=1200
  nonFocusVolumeAttenuation=0.5
  stereoSpread=500
  attenuationShapeExtents {x=0}
 }
 concurrencyOverrides {bLimitToOwner=true, maxCount=1, resolutionRule=PreventNew}
}
