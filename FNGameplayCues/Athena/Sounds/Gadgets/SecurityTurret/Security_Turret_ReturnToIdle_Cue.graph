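# Sound cue for the security turret gadget returning to idle, layering a mechanical rotate sound with a scan-beam shutdown.
# A mixer combines a random pick of six camera move-rotate waves, at reduced level, with a random pick of three scan-beam stop waves pitched slightly up by a modulator.
# It is spatialized by the turret sphere attenuation in the gadgets sound class, and its concurrency override allows one instance, stopping the oldest.
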SoundCue /FNGameplayCues/Athena/Sounds/Gadgets/SecurityTurret/Security_Turret_ReturnToIdle_Cue
root=Mixer_0
nodes [
 {id=Mixer_0, children [Random_0,Modulator_0], properties {inputVolume [0.7,1]}}
 {id=Modulator_0, children [Random_2], properties {pitchMax=1.2, pitchMin=1.1}}
 {id=Random_0, children [WavePlayer_1,WavePlayer_2,WavePlayer_3,WavePlayer_4,WavePlayer_5,WavePlayer_6], properties {weights [1,1,1,1,1,1]}}
 {id=Random_2, children [WavePlayer_7,WavePlayer_8,WavePlayer_9], properties {weights [1,1,1]}}
 {id=WavePlayer_1, properties {soundWaveAssetPtr=/Game/Athena/Sounds/Gadgets/SecurityTurret/Security_Camera_MoveRotate_01}}
 {id=WavePlayer_2, properties {soundWaveAssetPtr=/Game/Athena/Sounds/Gadgets/SecurityTurret/Security_Camera_MoveRotate_02}}
 {id=WavePlayer_3, properties {soundWaveAssetPtr=/Game/Athena/Sounds/Gadgets/SecurityTurret/Security_Camera_MoveRotate_03}}
 {id=WavePlayer_4, properties {soundWaveAssetPtr=/Game/Athena/Sounds/Gadgets/SecurityTurret/Security_Camera_MoveRotate_04}}
 {id=WavePlayer_5, properties {soundWaveAssetPtr=/Game/Athena/Sounds/Gadgets/SecurityTurret/Security_Camera_MoveRotate_05}}
 {id=WavePlayer_6, properties {soundWaveAssetPtr=/Game/Athena/Sounds/Gadgets/SecurityTurret/Security_Camera_MoveRotate_06}}
 {id=WavePlayer_7, properties {soundWaveAssetPtr=/Game/Athena/Sounds/Gadgets/SecurityTurret/Security_Turret_ScanBeam_Stop_01}}
 {id=WavePlayer_8, properties {soundWaveAssetPtr=/Game/Athena/Sounds/Gadgets/SecurityTurret/Security_Turret_ScanBeam_Stop_02}}
 {id=WavePlayer_9, properties {soundWaveAssetPtr=/Game/Athena/Sounds/Gadgets/SecurityTurret/Security_Turret_ScanBeam_Stop_03}}
]
properties {
 attenuationSettings=/Game/Athena/Sounds/Gadgets/SecurityTurret/Security_Turret_Sphere_Att
 bOverrideConcurrency=true
 maxDistance=3500
 soundClassObject=&Fort_Gadgets
 volumeMultiplier=1.5
 concurrencyOverrides {maxCount=1, resolutionRule=StopOldest}
}
