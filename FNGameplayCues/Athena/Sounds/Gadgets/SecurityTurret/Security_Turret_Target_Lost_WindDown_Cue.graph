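# Sound cue for the security turret gadget winding down after losing its target.
# A mixer layers a quieter, briefly delayed turret charge-down wave over a random pick of three security camera power-off waves pitched lower.
# It plays in the gadgets sound class with turret attenuation, and overridden concurrency keeps one instance per owner, stopping the oldest.
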SoundCue /FNGameplayCues/Athena/Sounds/Gadgets/SecurityTurret/Security_Turret_Target_Lost_WindDown_Cue
root=Mixer_0
nodes [
 {id=Delay_0, children [WavePlayer_3], properties {delayMax=0.15, delayMin=0.15}}
 {id=Mixer_0, children [Delay_0,Modulator_1], properties {inputVolume [0.4,1]}}
 {id=Modulator_1, children [Random_1], properties {pitchMax=0.8, pitchMin=0.8}}
 {id=Random_1, children [WavePlayer_8, WavePlayer_9, WavePlayer_10], properties {weights [1,1,1]}}
 {id=WavePlayer_10, properties {soundWaveAssetPtr=&/Security_Camera_PowerOff_03}}
 {id=WavePlayer_3, properties {soundWaveAssetPtr=&/Security_Turret_ChargeDown_02}}
 {id=WavePlayer_8, properties {soundWaveAssetPtr=&/Security_Camera_PowerOff_01}}
 {id=WavePlayer_9, properties {soundWaveAssetPtr=&/Security_Camera_PowerOff_02}}
]
properties {
 attenuationSettings=/Game/Athena/Sounds/Gadgets/SecurityTurret/Security_Turret_Sphere_Att
 bOverrideConcurrency=true
 maxDistance=3500
 soundClassObject=&Fort_Gadgets
 volumeMultiplier=0.7
 concurrencyOverrides {bLimitToOwner=true, maxCount=1, resolutionRule=StopOldest}
}
